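# Sound cue for firing the Figment suppressed auto heavy pistol, built from reused suppressed SMG fire samples.
# A mixer layers a close branch on the third-person fire sound class with a slightly quieter medium-range branch on the distant fire class.
# Each branch randomly picks one of two waves, then applies modulation and its own attenuation, with the medium branch pitched up.
# Global and local gunfire concurrency sets govern voice limiting.
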SoundCue /Figment_Pistol_AutoHeavySuppressed_Athena/Sounds/SC_Weapon_Pistol_AutoHeavySuppressed_Figment_Cue
root=Mixer_3
nodes [
 {id=Attenuation_3, children [Modulator_1], properties {attenuationSettings=&Fort_Weapon_Projectile_Close_Athena_Att}}
 {
  id=Attenuation_5
  children [Modulator_2]
  properties {attenuationSettings=&Fort_Weapon_Projectile_Medium_Mono_Athena_Att}
 }
 {id=Mixer_3, children [SoundClass_3,SoundClass_5], properties {inputVolume [1,0.8]}}
 {id=Modulator_1, children [Random_2]}
 {id=Modulator_2, children [Random_0], properties {pitchMax=1.2, pitchMin=1.1}}
 {id=Random_0, children [WavePlayer_0,WavePlayer_1], properties {weights [1,1]}}
 {id=Random_2, children [WavePlayer_2,WavePlayer_3], properties {weights [1,1]}}
 {
  id=SoundClass_3
  children [Attenuation_3]
  properties {soundClassOverride=&Fort_WeaponRanged_Fire_3P_PSM}
 }
 {
  id=SoundClass_5
  children [Attenuation_5]
  properties {soundClassOverride=&Fort_WeaponRanged_Fire_Distant}
 }
 {id=WavePlayer_0, properties {soundWaveAssetPtr=&/SMG_Suppressed_Fire_Medium_01}}
 {id=WavePlayer_1, properties {soundWaveAssetPtr=&/SMG_Suppressed_Fire_Medium_02}}
 {id=WavePlayer_2, properties {soundWaveAssetPtr=&/SMG_Suppressed_Fire_02}}
 {id=WavePlayer_3, properties {soundWaveAssetPtr=&/SMG_Suppressed_Fire_03}}
]
properties {
 bDebug=true
 maxDistance=10579.2871
 soundClassObject=&Fort_WeaponRanged_Fire_3P_PSM
 volumeMultiplier=1.2
 concurrencyOverrides {bLimitToOwner=true, maxCount=3, resolutionRule=StopOldest}
 concurrencySet [&Weapon_Gunfire_Global_Conc, &Weapon_Gunfire_Local_Conc]
}
